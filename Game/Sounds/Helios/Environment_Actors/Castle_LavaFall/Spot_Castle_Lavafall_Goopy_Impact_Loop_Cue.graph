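# Looping spot ambience cue for the goopy impact of the castle lavafall, built from reused brimstone waterfall recordings.
# A mixer layers a waterfall-bottom loop with a quieter, randomly chosen flowing-water thickener loop, then a modulator adds slight pitch variation.
# All quality levels share that chain.
# It uses occlusion-named waterfall attenuation, an indoor ambient sound class, exterior ambient bus and reverb sends, and tightly limited concurrency.
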SoundCue /Game/Sounds/Helios/Environment_Actors/Castle_LavaFall/Spot_Castle_Lavafall_Goopy_Impact_Loop_Cue
root=QualityLevel_0
nodes [
 {id=Mixer_0, children [Random_0,WavePlayer_0], properties {inputVolume [0.5,1]}}
 {id=Modulator_1, children [Mixer_0], properties {pitchMax=1.1, pitchMin=0.9, volumeMax=1, volumeMin=1}}
 {id=QualityLevel_0, children [Modulator_1,Modulator_1,Modulator_1,Modulator_1]}
 {id=Random_0, children [WavePlayer_3,WavePlayer_4,WavePlayer_5], properties {weights [1,1,1]}}
 {
  id=WavePlayer_0
  properties {bLooping=true, soundWaveAssetPtr=/Game/Sounds/Helios/Environment_Actors/BrimstoneWaterfall/Spot_Brimstone_Waterfall_Bottom_Loop}
 }
 {
  id=WavePlayer_3
  properties {bLooping=true, soundWaveAssetPtr=/Game/Sounds/Helios/Environment_Actors/BrimstoneWaterfall/Spot_Water_Flowing_Thickener_Loop_01}
 }
 {
  id=WavePlayer_4
  properties {bLooping=true, soundWaveAssetPtr=/Game/Sounds/Helios/Environment_Actors/BrimstoneWaterfall/Spot_Water_Flowing_Thickener_Loop_02}
 }
 {
  id=WavePlayer_5
  properties {bLooping=true, soundWaveAssetPtr=/Game/Sounds/Helios/Environment_Actors/BrimstoneWaterfall/Spot_Water_Flowing_Thickener_Loop_03}
 }
]
properties {
 attenuationSettings=/Game/Sounds/Helios/Environment_Actors/BrimstoneWaterfall/Spot_Water_Waterfall_Goopy_Impact_Heavy_Occ_Att
 bOverrideConcurrency=true
 maxDistance=2000
 priority=10
 soundClassObject=&Fort_Ambient_Indoors_Athena
 volumeMultiplier=0.501
 busSends [{audioBus=null, minSendLevel=1, sendLevel=1, soundSourceBus=&AmbientAudio_Exterior_SourceBus}]
 concurrencyOverrides {maxCount=2, voiceStealReleaseTime=2}
 soundSubmixSends [{minSendLevel=1, soundSubmix=&Fortnite_Reverb_Default}]
}
